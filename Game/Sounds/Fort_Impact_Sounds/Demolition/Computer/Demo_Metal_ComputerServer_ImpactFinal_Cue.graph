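# Sound cue for the final demolition impact of a metal computer server.
# A single wave, Computer_FinalImpact_01, plays through a Modulator node at its usual settings for pitch and volume variation.
# The cue uses the impacts sound class, a medium-sized object destruction attenuation and a shared object-impact concurrency set to control mixing, falloff and simultaneous voices.
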SoundCue /Game/Sounds/Fort_Impact_Sounds/Demolition/Computer/Demo_Metal_ComputerServer_ImpactFinal_Cue
root=Modulator_0
nodes [
 {id=Modulator_0, children [WavePlayer_0]}
 {id=WavePlayer_0, properties {soundWaveAssetPtr=&/Computer_FinalImpact_01}}
]
properties {
 attenuationSettings=/Game/Sounds/Attenuation/Fort_Demolition_Destroy_Medium_Sized_Object_Att
 maxDistance=4000
 soundClassObject=&Fort_Impacts_PSM
 volumeMultiplier=0.544
 concurrencyOverrides {maxCount=3, resolutionRule=StopOldest}
 concurrencySet [&CONC_Object_Impacts_FourMax]
}
